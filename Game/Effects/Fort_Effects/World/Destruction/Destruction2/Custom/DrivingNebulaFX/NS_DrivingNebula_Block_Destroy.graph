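# Niagara system that throws a burst of tinted mesh debris when a DrivingNebula block is destroyed.
# One GPU emitter fires two randomly sized bursts as separate spawn groups, placed within User.BlockSize boxes and colored by User.BlockColor.
# Particle partitions pick among three mesh indices and scales, and spawn group varies brightness.
# Debris launches outward under ramped gravity and drag, collides, tumbles until first impact, then shrinks away late in life.
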NiagaraSystem /Game/Effects/Fort_Effects/World/Destruction/Destruction2/Custom/DrivingNebulaFX/NS_DrivingNebula_Block_Destroy
dependencies {
 usedDataInterfaces [
  &NiagaraDataInterfaceArrayInt32
  &NiagaraDataInterfaceArrayFloat3
  &NiagaraDataInterfaceMeshRendererInfo
  &NiagaraDataInterfaceCurve
  &NiagaraDataInterfaceArrayFloat
 ]
 usedDynamicInputs []
 usedModules [&SystemState]
}
emitters {Resource {
  simTarget=GPUComputeSim
  emitterUpdateScript [
   &EmitterState()
   &SpawnBurst_Instantaneous(
    Spawn Count=&RandomRangeInt(Minimum=15, Maximum=25)
    Spawn Group=1
    Use Spawn Probability=false
    UseLoopCountLimit=false
   ) {module=SpawnBurst_Instantaneous001}
   &SpawnBurst_Instantaneous(
    Spawn Count=&RandomRangeInt(Minimum=35, Maximum=45)
    Use Spawn Probability=false
    UseLoopCountLimit=false
   )
  ]
  particleSpawnScript [
   /SetVariables(
    Particles.BlockType:NiagaraInt32=&SelectIntFromArray(
     Direct Array Index=&MakeIntFromFloat(Float=$User.BlockPicker)
     Int Selection Array=DataInterface() {propertyValues {IntData [0,1,2,3,4,5,6]}}
    )
    Particles.BlockVis:NiagaraInt32=&SpawnGroup()
   ) {isSetParameters=true}
   &PartitionParticles(Number Of Partitions=3)
   &InitializeParticle(
    Color=$User.BlockColor
    Lifetime Mode=::NewEnumerator1
    Mass Mode=::NewEnumerator1
    Mesh Renderer Array Visibility Mode=::NewEnumerator1
    Mesh Renderer Info=DataInterface() {propertyValues {MeshRenderer=&self:Resource_0.NiagaraMeshRendererProperties_8}}
    Mesh Renderer Mesh Index:NiagaraInt32=&SelectIntFromArray(Direct Array Index=$Particles.PartitionParticles.Partition, Int Selection Array=DataInterface() {propertyValues {IntData [0,1,2]}})
    Mesh Scale=&SelectVectorFromArray(
     Direct Array Index=$Particles.PartitionParticles.Partition
     Vector Selection Array=DataInterface() {propertyValues {FloatData [<3 3 3>, <0.5 0.5 0.5>, <0.7 0.5 0.5>]}}
    )
    Mesh Scale Mode=::NewEnumerator4
    Position Mode=::NewEnumerator0
    Write Material Random=true
   )
   &ShapeLocation(
    Additional Quaternion Rotation=false
    Box Size=$User.BlockSize
    MaskSpawnGroup=false
    Offset=$User.FXOffset
    Offset Mode=::NewEnumerator0
    Override Seed=false
    Rotation Mode=::NewEnumerator3
    Shape Primitive=::NewEnumerator2
    Surface Only=false
    UseOwnerScale=false
   )
   &ShapeLocation(
    Additional Quaternion Rotation=false
    Box Size=$User.BlockSize
    MaskSpawnGroup=true
    Override Seed=false
    Rotation Mode=::NewEnumerator3
    Shape Primitive=::NewEnumerator2
    Spawn Group Mask=1
    Surface Only=false
    UseOwnerScale=false
   ) {module=ShapeLocation001}
   &InitialMeshOrientation(Use Rotation Vector=false)
   &AddVelocity(
    Module.Use Falloff Distance=false
    Override Seed=false
    Velocity Mode=::NewEnumerator1
    Velocity Speed=&RandomRangeFloat(Minimum=250, Maximum=335)
   )
  ]
  particleUpdateScript [
   /SetVariables(Particles.BlockScale:Vector3f=&VectorFromFloat(Value=&FloatFromCurve(
      FloatCurve=DataInterface() {propertyValues {Curve {keys [
          <at 0 0 RCIM_Constant>
          <at 0.05 1 RCIM_Constant>
          {interpMode=RCIM_Cubic, leaveTangentWeight=0.3333, tangentMode=RCTM_User, time=0.75, value=1}
          {arriveTangentWeight=0.3333, interpMode=RCIM_Cubic, tangentMode=RCTM_User, time=1, value=0}
         ]}}}
      Scale Curve=&SelectFloatFromArray(Direct Array Index=$Particles.BlockType, Float Selection Array=DataInterface() {propertyValues {FloatData [0.2,0.2,0.2,0.2,0.2,0.2,0.2]}})
     ))) {isSetParameters=true}
   &ParticleState()
   &GravityForce(Gravity=&Multiply_VectorByFloat(
     Vector=<0 0 -980>
     Float=&FloatFromCurve(FloatCurve=DataInterface() {propertyValues {Curve {
         defaultValue=1
         keys [<at 0 0 RCIM_Cubic>, {arriveTangent=30.1745, interpMode=RCIM_Cubic, leaveTangent=30.1745, tangentMode=RCTM_User, time=0.1, value=1}]
        }}})
    ))
   &Drag(
    Drag=&self:Gravity_Drag_Compensation_02(Drag Amount:NiagaraFloat=1, Drag Falloff:NiagaraFloat=2, Gravity Direction:Vector3f=<0 0 -1>)
    Use Linear Drag=true
    Use Rotational Drag=true
   )
   &Collision(
    Advanced Aging Rate=1
    Age Colliding Particles=false
    Enable Rest State=false
    Friction During a Bounce=0.2
    GPU Collision Type=::NewEnumerator2
    Kill On Collision=2
    Kill Particles Lodged Within Meshes=false
    KillOnCollision=false
    Mesh Dimensions:Vector3f=<50 50 50>
    Module.EnableMaxCollisionCount=false
    Module.Randomize Collision Normal Vector=true
    Particle Radius Scale=0.1
    Radius Calculation Type=::NewEnumerator1
    Restitution=0.15
   )
   &SolveForcesAndVelocity(Clamp Velocity=false, Limit Acceleration=false)
   &UpdateMeshOrientation(Rotation Vector=&MakeCustomVectorFromBool(
     Bool=$Particles.HasCollided
     True Vector=<0 0 0>
     False Vector=&RandomRangeVector(Minimum=<-1 -1 -1>, Maximum=<1 1 1>)
    ))
   &ScaleMeshSize(Scale Factor=&VectorFromFloat(Value=&FloatFromCurve(FloatCurve=DataInterface() {propertyValues {Curve {keys [
          <at 0 0 RCIM_Constant>
          <at 0.05 1 RCIM_Constant>
          {interpMode=RCIM_Cubic, leaveTangentWeight=0.3333, tangentMode=RCTM_User, time=0.75, value=1}
          {arriveTangentWeight=0.3333, interpMode=RCIM_Cubic, tangentMode=RCTM_User, time=1, value=0}
         ]}}})))
   &ScaleColor(
    Scale RGB=&MaskVectorBySpawnGroup(Masked Vector=<0.5 0.5 0.5>, Passthrough Vector=<0.9 0.9 0.9>)
    ScaleA=true
    ScaleRGB=true
   )
  ]
  rendererClasses [&NiagaraMeshRendererProperties]
  renderers [{rendererClass=&NiagaraMeshRendererProperties, rendererIndex=0}]
 }}
summary {
 systemName=NS_DrivingNebula_Block_Destroy
 emitters [{emitterName=Resource, simTarget=GPUComputeSim, rendererClasses [&NiagaraMeshRendererProperties]}]
 userVariables [
  User.bIsCurie {type=NiagaraBool}
  User.bIsInstantDeath {type=NiagaraBool}
  User.BlockColor {type=LinearColor, description="", defaultValue {struct=LinearColor, value=<rgba 0.0016 0.4433 0.5506 1>}}
  User.BlockPicker {type=NiagaraFloat, defaultValue=0, description=""}
  User.BlockSize {type=Vector3f, description="", defaultValue {struct=Vector3f, value=<200 200 1200>}}
  User.FXOffset {type=Vector3f, description="", defaultValue {struct=Vector3f, value=<0 0 100>}}
  User.LeafAmount {type=NiagaraFloat, defaultValue=1500, description=""}
 ]
}
